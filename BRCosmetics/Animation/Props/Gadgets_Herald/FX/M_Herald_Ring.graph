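# Unlit translucent ring material for Herald gadget FX, usable on Niagara sprites.
# An outer 2D sphere mask multiplied by an inverted inner mask forms the ring, shaped by particle dynamic parameters.
# Particle color tints the ring through an emissive adjustment function.
# Particle alpha times the ring mask blends a lerp from 1 toward a second dynamic parameter's Refraction value, suggesting a distortion effect.
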Material /BRCosmetics/Animation/Props/Gadgets_Herald/FX/M_Herald_Ring
graph {
 properties {bUsedWithNiagaraSprites=true, blendMode=BLEND_Translucent, pixelDepthOffsetMode=PDOM_Legacy, shadingModel=MSM_Unlit, usedShadingModels=Unlit}
 nodes [
  Divide()
  n1 = DynamicParameter() {parameterIndex=1, defaultValue=<rgba 0.25 1 1 1>, paramNames [InnerRing,Refraction,Param3,Param4]}
  n2 = ParticleColor()
  n3 = DynamicParameter() {defaultValue=<rgba 0.2 0.9 0.3 0.9>, paramNames [InnerRing, InnerHardness, OuterRing, OuterHardness]}
  n4 = Multiply(
   A=MaterialFunctionCall!(Hardness=@n3.OuterHardness, Center, Radius=@n3.OuterHardness) {materialFunction=/Game/Packages/Fortress_SharedMaterials/Base_Material_Functions/FX_SphereMask2D}.Result
   B=OneMinus(None=MaterialFunctionCall(Hardness=@n3.InnerHardness, Radius=@n3.InnerHardness) {materialFunction=&FX_SphereMask2D}.Result)
  )
  LinearInterpolate(B=@n1.Refraction, Alpha=Multiply(A=@n2.A, B=@n4)) {constA=1, constB=1.5}
  MaterialFunctionCall(Emissive=Multiply(A=@n2.RGB, B=@n4)) {materialFunction=&MF_FX_EmissiveAdjustments}
  ScalarParameter() {defaultValue=0.25, parameterName=Radius}
 ]
}
