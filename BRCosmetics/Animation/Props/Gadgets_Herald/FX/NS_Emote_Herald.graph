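# Niagara system for the Herald gadget emote prop, layering a blue energy glow from four emitters.
# EnergyDots spawns small sphere-placed sparks that partly inherit source movement, swirl under curl noise and drag, then fade and shrink.
# Rays spawns growing, fading sprites animated by a radial material module with randomized tiling and spin.
# Light_Render and Light_Ring burst at staggered times, adding a light and a material-parameter-driven ring sprite.
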NiagaraSystem /BRCosmetics/Animation/Props/Gadgets_Herald/FX/NS_Emote_Herald
dependencies {usedDataInterfaces [&NiagaraDataInterfaceCurve], usedDynamicInputs [], usedModules [&SystemState]}
emitters {
 EnergyDots {
  emitterUpdateScript [
   &EmitterState(Life Cycle Mode=::NewEnumerator1, Loop Delay=0.12, Loop Duration=1, UseLoopDelay=true)
   &SpawnRate(Spawn Probability=0.5, SpawnRate=&RandomRangeFloat(Minimum=3, Maximum=7), Use Spawn Probability=false)
  ]
  particleSpawnScript [
   &InitializeParticle(
    Color=&MakeCustomLinearColorFromBool(Bool=&RandomBool(), True Color=<rgba 0.1911 1.1333 3 1>, False Color=<rgba 0 0.3027 3 1>)
    Lifetime Max=0.5
    Lifetime Min=0.4
    Lifetime Mode=::NewEnumerator1
    Sprite Rotation Angle=&RandomRangeFloat(Minimum=0, Maximum=359)
    Sprite Rotation Mode=::NewEnumerator0
    Sprite Size Mode=::NewEnumerator1
    Uniform Sprite Size Max=3
    Uniform Sprite Size Min=1
    UsePositionOffset=false
    Write Material Random=true
   ) {module=InitializeParticle001}
   &ShapeLocation(MaskSpawnGroup=false, Override Seed=false, Sphere Radius=5, Sphere Surface Distribution=1, UseOwnerScale=true)
  ]
  particleUpdateScript [
   &ParticleState()
   &InheritSourceMovement(Applied Position Scale=<0.2 0.2 0.2>, ApplyForce=false, ApplyPosition=true, ApplyPositionFalloff=false)
   &Drag(Drag=5, Use Linear Drag=true, Use Rotational Drag=false)
   &CurlNoiseForce(Noise Frequency=25, Noise Strength=150, Pan Noise=false)
   &ScaleColor(
    Scale Alpha=&FloatFromCurve(FloatCurve=DataInterface() {propertyValues {Curve {keys [
         {arriveTangent=9.0725, interpMode=RCIM_Cubic, leaveTangent=9.0725, tangentMode=RCTM_User, time=0, value=0}
         {arriveTangent=1.5784, interpMode=RCIM_Cubic, leaveTangent=1.5784, tangentMode=RCTM_User, time=0.1408, value=0.9847}
         {arriveTangent=-0.0405, interpMode=RCIM_Cubic, leaveTangent=-0.0405, tangentMode=RCTM_User, time=1, value=0}
        ]}}})
    ScaleA=true
    ScaleRGB=false
   )
   &SolveForcesAndVelocity(Clamp Velocity=false, Limit Acceleration=false)
   &ScaleSpriteSize(Uniform Curve Sprite Scale=DataInterface() {propertyValues {Curve {keys [
        {tangentMode=RCTM_User, time=0, value=1}
        {arriveTangent=-0.0405, interpMode=RCIM_Cubic, leaveTangent=-0.0405, tangentMode=RCTM_User, time=0.7009, value=1}
        {arriveTangent=-8.0089, interpMode=RCIM_Cubic, leaveTangent=-8.0089, tangentMode=RCTM_User, time=1, value=0}
       ]}}})
  ]
  rendererClasses [&NiagaraSpriteRendererProperties]
  renderers [{rendererClass=&NiagaraSpriteRendererProperties}]
 }
 Light_Render {
  emitterUpdateScript [
   &EmitterState(Life Cycle Mode=::NewEnumerator1, Loop Duration=2, UseLoopDelay=false)
   &SpawnBurst_Instantaneous(Spawn Time=0.45, Use Spawn Probability=false, UseLoopCountLimit=false)
  ]
  particleSpawnScript [
   &InitializeParticle(
    Color=<rgba 0.4276 1.0995 2 0.7>
    Lifetime=2
    Sprite Size Mode=::NewEnumerator3
    Uniform Sprite Size=22
    UsePositionOffset=false
    Write Material Random=true
   )
   /SetVariables(Particles.LightColor=<rgba 0.1922 0.6941 1 2>) {isSetParameters=true}
  ]
  particleUpdateScript [&ParticleState()]
  rendererClasses [&NiagaraLightRendererProperties]
  renderers [{rendererClass=&NiagaraLightRendererProperties, rendererIndex=0}]
 }
 Light_Ring {
  emitterUpdateScript [
   &EmitterState(Life Cycle Mode=::NewEnumerator1, Loop Duration=2, UseLoopDelay=false)
   &SpawnBurst_Instantaneous(Spawn Time=0.72, Use Spawn Probability=false, UseLoopCountLimit=false)
  ]
  particleSpawnScript [&InitializeParticle(
    Color=<rgba 0.4276 1.0995 2 0.7>
    Lifetime=2
    Sprite Size Mode=::NewEnumerator3
    Uniform Sprite Size=22
    UsePositionOffset=false
    Write Material Random=true
   )]
  particleUpdateScript [
   &ParticleState()
   &DynamicMaterialParameters(
    Index 0 Param 1=0.3
    Index 0 Param 2=0.95
    Index 0 Param 3=0.42
    Index 0 Param 4=0.3
    Module.Param0WriteEnabled=true
    Module.Param1WriteEnabled=true
    Module.Param2WriteEnabled=true
    Module.Param3WriteEnabled=true
   )
  ]
  rendererClasses [&NiagaraSpriteRendererProperties]
  renderers [{rendererClass=&NiagaraSpriteRendererProperties}]
 }
 Rays {
  emitterUpdateScript [
   &EmitterState()
   &SpawnRate(SpawnRate=2, Use Spawn Probability=false)
  ]
  particleSpawnScript [&InitializeParticle(
    Color=<rgba 0.1719 0.6146 1 1>
    Lifetime=1.25
    Sprite Rotation Mode=::NewEnumerator2
    Sprite Size Mode=::NewEnumerator3
    Uniform Sprite Size=27
    UsePositionOffset=false
    Write Material Random=true
   )]
  particleUpdateScript [
   &ParticleState()
   &ScaleSpriteSize(Uniform Curve Sprite Scale=DataInterface() {propertyValues {Curve {keys [{interpMode=RCIM_Cubic, tangentMode=RCTM_User, time=0, value=0.7}, {interpMode=RCIM_Cubic, tangentMode=RCTM_User, time=1, value=1.7}]}}})
   &FN_Radial_Material(
    Base Sprite Tile U:NiagaraFloat=&RandomRangeFloat(Minimum=1.2, Maximum=2)
    Radial Mask Power=3
    Radial Mask Strength=0.8
    Sprite Extrusion Speed=0.5
    Sprite Spin Speed:NiagaraFloat=&RandomRangeFloat(Minimum=0.2, Maximum=0.5)
   )
   &ScaleColor(
    Scale Alpha=&FloatFromCurve(FloatCurve=DataInterface() {propertyValues {Curve {keys [{interpMode=RCIM_Cubic, tangentMode=RCTM_User, time=0, value=1}, {interpMode=RCIM_Cubic, tangentMode=RCTM_User, time=1, value=0}]}}})
    ScaleA=true
    ScaleRGB=false
   )
  ]
  rendererClasses [&NiagaraSpriteRendererProperties]
  renderers [{rendererClass=&NiagaraSpriteRendererProperties}]
 }
}
summary {
 systemName=NS_Emote_Herald
 emitters [
  {emitterName=EnergyDots, rendererClasses [&NiagaraSpriteRendererProperties]}
  {emitterName=Light_Render, rendererClasses [&NiagaraLightRendererProperties]}
  {emitterName=Rays, rendererClasses [&NiagaraSpriteRendererProperties]}
  {emitterName=Light_Ring, rendererClasses [&NiagaraSpriteRendererProperties]}
 ]
 userVariables [User.bIsFrontEndPreview {type=NiagaraBool}]
}
